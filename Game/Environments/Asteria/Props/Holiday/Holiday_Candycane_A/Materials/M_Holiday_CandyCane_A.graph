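# Nanite-ready Default Lit material for the Holiday Candycane A prop, built on shared WinterArt layer functions.
# One layer of diffuse, spec-mask and normal textures, with height-masked tint and optional two-sided sets, feeds a four-layer lerp function.
# Emissive switches between neon-sign and channel-mask functions, with the masked color saturated for ray tracing.
# Final emissive is scaled by EmissiveEnabled and an axis-aligned Fresnel, which becomes 1 under ray tracing.
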Material /Game/Environments/Asteria/Props/Holiday/Holiday_Candycane_A/Materials/M_Holiday_CandyCane_A
graph {
 properties {
  bUseMaterialAttributes=true
  bUsedWithNanite=true
  pixelDepthOffsetMode=PDOM_Legacy
  refractionMethod=RM_IndexOfRefraction
  usedShadingModels=Default Lit
 }
 nodes [
  n1 = MaterialFunctionCall() {materialFunction=/Game/Athena/WinterArt/Materials/MF_UVsSet}
  n2 = VectorParameter() {
   group=Material_1
   parameterName=L1_UVScaleByComponent
   sortPriority=3
   channelNames {a="", b=Normals, g=Specular, r=Diffuse}
   defaultValue=<rgba 1 1 1 0>
  }
  n3 = TextureSampleParameter2D(UVs=Multiply(A=@n1.UVs Layer 1, B=@n2.Diffuse)) {
   group=Material_1
   parameterName=Diffuse Side 2
   samplerSource=SSM_Wrap_WorldGroupSettings
   sortPriority=1
   texture=/Game/Global/Textures/Default/Blanks/T_MidgreysRGB
  }
  n4 = TwoSidedSign()
  n5 = VectorParameter() {
   group=Material_1
   parameterName=L1_UVScaleByComponent
   sortPriority=3
   channelNames {a="", b=Normals, g=Specular, r=Diffuse}
   defaultValue=<rgba 1 1 1 0>
  }
  n6 = Multiply(A=@n1.UVs Layer 1, B=@n5.Diffuse)
  n7 = StaticSwitchParameter(True=Multiply(A=ScalarParameter() {defaultValue=1, group=Material_1, parameterName=L1_UV_ColorSwatchScaleAdd}, B=@n6), False=@n6) {group=Material_1, parameterName=Use Color Swatch}
  n8 = TextureSampleParameter2D(UVs=Multiply(A=ScalarParameter() {defaultValue=1, group=Material_1, parameterName=L1_CustomHeightScale, sortPriority=1}, B=@n7)) {
   group=Material_1
   parameterName=L1_Height
   samplerSource=SSM_Wrap_WorldGroupSettings
   samplerType=SAMPLERTYPE_LinearGrayscale
   sortPriority=-5
   texture=/Game/Global/Textures/Default/Blanks/T_BlackLinearGrayscale
  }
  n9 = TextureSampleParameter2D(UVs=@n6) {
   group=Material_1
   parameterName=Background Diffuse
   samplerSource=SSM_Wrap_WorldGroupSettings
   sortPriority=-5
   texture=&T_WhitesRGB
  }
  n10 = TextureSampleParameter2D(UVs=@n7) {
   group=Material_1
   parameterName=Diffuse
   samplerSource=SSM_Wrap_WorldGroupSettings
   sortPriority=-5
   texture=&T_MidgreysRGB
  }
  n11 = StaticSwitchParameter(True=@n8.R, False=StaticSwitchParameter(True=@n9.A, False=@n10.A) {group=Material_1, parameterName=Use Color Swatch}) {desc=Use Smale scale grayscale, group=Material_1, parameterName=L1 Custom Height, sortPriority=1}
  n12 = Multiply(A=@n5.Specular, B=@n1.UVs Layer 1)
  n13 = TextureSampleParameter2D(UVs=@n12) {
   group=Material_1
   parameterName=SpecularMasks
   samplerSource=SSM_Wrap_WorldGroupSettings
   samplerType=SAMPLERTYPE_Masks
   sortPriority=-5
   texture=&T_Fortnite_Default_S
  }
  n14 = MaterialFunctionCall!(
   Transition=ScalarParameter() {defaultValue=0.5, group=Material_1, parameterName=L1_TintTransition, sortPriority=5}
   HM=@n11
   Contrast=ScalarParameter() {defaultValue=5, group=Material_1, parameterName=L1_TintContrast, sortPriority=5}
   Metalness=@n13.G
   BG Color=StaticSwitchParameter(True=@n9.RGB, False=@n10.RGB) {group=Material_1, parameterName=Use Color Swatch}
   TintColor=StaticSwitchParameter(True=@n10.RGB, False=VectorParameter() {group=Material_1, parameterName=TintColor, sortPriority=7, defaultValue=<rgba 1 1 1 0>}.RGB) {group=Material_1, parameterName=Use Color Swatch}
  ) {materialFunction=/Game/Athena/WinterArt/Materials/MF_ColorMask}
  n15 = AppendVector(
   A=StaticSwitchParameter(True=@n14.LerpMix, False=@n14.LerpMix) {group=Material_1, parameterName=L1_UseTintColorAsDiffuse}
   B=Multiply(A=@n11, B=ScalarParameter() {defaultValue=1, group=Material_1, parameterName=L1 Alpha, sortPriority=1})
  )
  n16 = TextureObjectParameter() {group=Material_1, parameterName=Diffuse, sortPriority=1, texture=&T_MidgreysRGB}
  n17 = ScalarParameter() {group=Adv Emissive, parameterName=L1 X Speed, sortPriority=12}
  n18 = ScalarParameter() {group=Adv Emissive, parameterName=L1 Y Speed, sortPriority=13}
  n19 = VectorParameter() {group=Adv Emissive, parameterName=L1 Channel, sortPriority=9}
  n20 = ScalarParameter() {defaultValue=1, group=Adv Emissive, parameterName=L1 Offset, sliderMax=4, sliderMin=1, sortPriority=11}
  n21 = VectorParameter() {
   desc=WARNING* Run through Saturate node when Lumen
   group=Material_1
   parameterName=Emissive
   sortPriority=7
   channelNames {a=Emissive Multiplier, b="", g="", r=""}
   defaultValue=<rgba 0 0 0 1>
  }
  n22 = Multiply(A=@n21, B=@n21)
  n23 = RayTracingQualitySwitch(Normal=@n22, RayTraced=Saturate(None=@n22)) {desc=For Lumen}
  n24 = Multiply(A=@n1.UVs Layer 1, B=@n2.Specular)
  n25 = TextureSampleParameter2D(UVs=@n24) {
   group=Material_1
   parameterName=SpecularMasks Side 2
   samplerSource=SSM_Wrap_WorldGroupSettings
   samplerType=SAMPLERTYPE_Masks
   sortPriority=2
   texture=&T_Fortnite_Default_S
  }
  n26 = TextureCoordinate() {coordinateIndex=1}
  n27 = TextureSampleParameter2D(UVs=StaticSwitchParameter(
    True=TextureCoordinate() {coordinateIndex=2}
    False=StaticSwitchParameter(True=@n26, False=@n24) {group=Material_1, parameterName=L1_Ch2Emissive Side2, sortPriority=-5}
   ) {group=Material_1, parameterName=L1_Ch3Emissive Side2, sortPriority=-5}) {
   group=Material_1
   parameterName=L1_Emissive Side 2
   samplerSource=SSM_Wrap_WorldGroupSettings
   samplerType=SAMPLERTYPE_LinearGrayscale
   sortPriority=-4
   texture=&T_BlackLinearGrayscale
  }
  n28 = AppendVector(A=@n25.RGB, B=StaticSwitchParameter(True=@n27.R, False=@n25.A) {group=Material_1, parameterName=L1_CustomEmissive Side 2, sortPriority=-5})
  n29 = TextureCoordinate() {coordinateIndex=1}
  n30 = TextureSampleParameter2D(UVs=StaticSwitchParameter(
    True=TextureCoordinate() {coordinateIndex=2}
    False=StaticSwitchParameter(True=@n29, False=@n12) {group=Material_1, parameterName=L1_Ch2Emissive, sortPriority=-5}
   ) {group=Material_1, parameterName=L1_Ch3Emissive, sortPriority=1}) {
   group=Material_1
   parameterName=L1_Emissive
   samplerSource=SSM_Wrap_WorldGroupSettings
   samplerType=SAMPLERTYPE_LinearGrayscale
   sortPriority=-5
   texture=&T_BlackLinearGrayscale
  }
  n31 = AppendVector(A=@n13.RGB, B=StaticSwitchParameter(True=@n30.R, False=@n13.A) {group=Material_1, parameterName=L1_CustomEmissive, sortPriority=-5})
  n32 = TextureSampleParameter2D(UVs=Multiply(A=@n2.Normals, B=@n1.UVs Layer 1)) {
   group=Material_1
   parameterName=Normals Side 2
   samplerSource=SSM_Wrap_WorldGroupSettings
   samplerType=SAMPLERTYPE_Normal
   sortPriority=3
   texture=/Engine/EngineMaterials/FlatNormal
  }
  n33 = TextureSampleParameter2D(UVs=Multiply(A=@n5.Normals, B=@n1.UVs Layer 1)) {
   group=Material_1
   parameterName=Normals
   samplerSource=SSM_Wrap_WorldGroupSettings
   samplerType=SAMPLERTYPE_Normal
   sortPriority=-5
   texture=&FlatNormal
  }
  n34 = ScalarParameter() {defaultValue=1, group=Material_1, parameterName=L1_BoostNormals, sortPriority=3}
  n35 = Multiply(A=@n33.RGB, B=AppendVector(A=AppendVector(A=@n34, B=@n34), B=Constant() {r=1}))
  n36 = MaterialFunctionCall!(
   Use3Materials
   Use4Materials
   Use2Materials
   D=StaticSwitchParameter(True=LinearInterpolate(A=@n3.RGBA, B=@n15, Alpha=@n4), False=@n15) {group=Material_1, parameterName=TwoSided, sortPriority=18}
   S=StaticSwitchParameter(True=LinearInterpolate(A=@n28, B=@n31, Alpha=@n4), False=@n31) {group=Material_1, parameterName=TwoSided, sortPriority=18}
   N=StaticSwitchParameter(True=LinearInterpolate(A=@n32.RGB, B=@n35, Alpha=@n4), False=@n35) {group=Material_1, parameterName=TwoSided, sortPriority=18}
   D2
   S2
   N2
   D3
   S3
   N3
   D4
   S4
   N4
   E1=StaticSwitchParameter(
    True=MaterialFunctionCall!(Base Diffuse=@n16) {materialFunction=/Game/Athena/WinterArt/Materials/MF_NeonSignEmissive}.Result
    False=MaterialFunctionCall!(X=@n17, Y=@n18, Dot=@n19.RGB, TextureOffset=@n20, EmissiveTexture=@n23) {materialFunction=/Game/Athena/WinterArt/Materials/MF_EmissiveMask}.EmissiveMasked
   ) {group=NeonEmissive, parameterName=UseNeonSignEmissive_Layer1}
   E2
   E3
   E4
   CCMaskByMetal_1
   CC1
   CCR_1
   CC2
   CCR_2
   CCMaskByMetal_2
   CC3
   CCR_3
   CCMaskByMetal_3
   CC4
   CCMaskByMetal_4
   CCR_4
   OpacityMaskCancel_1
   OpacityMaskCancel_2
   OpacityMaskCancel_3
   OpacityMaskCancel_4
   SSS_1
   SSS_2
   SSS_3
   SSS_4
   SSS_Depth_1
   SSS_Depth_2
   SSS_Depth_3
   SSS_Depth_4
  ) {materialFunction=/Game/Athena/WinterArt/Materials/MF_FN_TextureSetLerp_4Layers}
  n37 = SetMaterialAttributes!(
   MaterialAttributes
   Base Color=@n36.Diffuse
   Specular=@n36.Diffuse
   Metallic=@n36.Diffuse
   Roughness=@n36.Diffuse
   Emissive Color=@n36.Diffuse
   Normal=@n36.Diffuse
   Custom Data 0=@n36.Diffuse
   Custom Data 1=@n36.Diffuse
   Opacity Mask=@n36.Diffuse
   Subsurface Color=@n36.Diffuse
   Opacity=@n36.Diffuse
   Refraction=NamedRerouteDeclaration(None=@n36.Spec_A) {name=Spec A, nodeColor=<rgba 0 0.2734 1 1>}
  ) {attributeSetTypes [
    #1
    #2
    #3
    #4
    #5
    #6
    #7
    #8
    #9
    #10
    #11
    #12
   ]}
  n38 = GetMaterialAttributes(None=@n37) {attributeGetTypes [#5]}
  n39 = ScalarParameter() {defaultValue=1, parameterName=EmissiveEnabled, sliderMax=1}
  n40 = Constant3Vector() {constant=<rgba 0 1 0 0>}
  n41 = RayTracingQualitySwitch(
   Normal=MaterialFunctionCall!(Axis=Transform(None=@n40), Exponent=ScalarParameter() {defaultValue=6, parameterName=AxisAlignedFresnelExponent}, Normal) {materialFunction=/Game/Effects/Fort_Effects/Materials/_Functions/MF_AxisAlignedFrensnel}.Fresnel
   RayTraced=Constant() {r=1}
  )
  SetMaterialAttributes!(MaterialAttributes=@n38.MaterialAttributes, Emissive Color=Multiply(A=@n38.Emissive Color, B=Multiply(A=@n39, B=@n41))) {attributeSetTypes [#5]}
 ]
}
